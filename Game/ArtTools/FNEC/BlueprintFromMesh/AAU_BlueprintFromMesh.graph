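# Editor utility that creates a prop Blueprint for each selected static mesh, parented to a chosen Small, Medium or Large blockout prop class or BuildingStaticGeo.
# CreateBlueprintFromMesh duplicates the BP_Prop_A template and writes the mesh to the new class default object.
# CreateNewAssetData derives each destination by swapping Meshes for Blueprints in the path and adjusting the name prefix, using BSG for BuildingStaticGeo.
# A second pass reparents the new Blueprints, then saves them.
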EditorUtilityBlueprint /Game/ArtTools/FNEC/BlueprintFromMesh/AAU_BlueprintFromMesh
graphs [
 CreateBlueprintFromMesh {nodes [
   n1 = CreateBlueprintFromMesh!() -> (then, ParentClass:ENUM_BPFromMesh_ClassSelection Enum=ENUM_BPFromMesh_ClassSelection_MAX) {is=title K2Node_FunctionEntry}
   > n2 = Switch on ENUM_BPFromMesh_ClassSelection!(execute, Selection:ENUM_BPFromMesh_ClassSelection Enum=@n1) -> (NewEnumerator0:Exec, NewEnumerator1:Exec, NewEnumerator2:Exec, NewEnumerator3:Exec) {is=title K2Node_SwitchEnum}
   > n3 = Set ParentClass(ParentClass:Object Class Reference=/Game/Environments/Blockout/Blueprints/Props/BP_Blockout_Prop_Small_A.BP_Blockout_Prop_Small_A_C) -> (Output_Get:Object Class Reference)
   n4 = GetSelectedAssets(execute=[@n19.then, @n3.then, @n17.then, @n18.then])
   > n5 = Filter by Class!(
    execute
    TargetArray:Array of Object References=@n4
    ObjectClass:Object Class Reference=&StaticMesh
    FilterType:EEditorScriptingFilterType Enum
   ) -> (then, ReturnValue:Array of Static Mesh Object References)
   > n6 = Set AR_SelectedMeshes(AR_SelectedMeshes:Array of Static Mesh Object References=@n5) -> (Output_Get:Array of Static Mesh Object References)
   > n7 = For Each Loop(Array:Array of Static Mesh Object References=@n6) -> (Array Element:Static Mesh Object Reference)
   n8 = For Each Loop(Exec=@n7.Completed, Array:Array of Object References=Get L_AR_NewAssets() -> (L_AR_NewAssets:Array of Object References)) -> (Array Element:Object Reference)
   SaveLoadedAssets(execute=@n8.Completed, AssetsToSave=Get L_AR_NewAssets(), bOnlyIfIsDirty=false)
   n9 = GetBlueprintAsset(execute=@n8.LoopBody, Object=@n8)
   > ReparentBlueprint(Blueprint=@n9, NewParentClass=Get ParentClass() -> (ParentClass:Object Class Reference))
   n10 = Set L_SM(execute=@n7.LoopBody, L_SM:Static Mesh Object Reference=@n7) -> (Output_Get:Static Mesh Object Reference)
   > n11 = CreateNewAssetData!(execute, self, InAsset:Object Reference=@n10) -> (then, PackageName:String)
   > n12 = DuplicateAsset!(execute, SourceAssetPath:String=/Game/Environments/Outsource/Templates/PropTemplate_A/Blueprints/BP_Prop_A, DestinationAssetPath:String=@n11) -> (then, ReturnValue:Object Reference)
   > n13 = Set L_NewAsset(L_NewAsset:Object Reference=@n12) -> (Output_Get:Object Reference)
   > n14 = GetBlueprintAsset(Object=@n13)
   > n15 = GeneratedClass!(execute, BlueprintObj:Blueprint Object Reference=@n14) -> (then, ReturnValue:Object Class Reference)
   > n16 = GetClassDefaultObject!(execute, Class:Object Class Reference=@n15) -> (then, ReturnValue:Object Reference)
   > SetEditorProperty(Object=@n16, PropertyName=StaticMesh, PropertyValue:Static Mesh Object Reference=Get L_SM() -> (L_SM:Static Mesh Object Reference))
   > Add Unique(
    TargetArray:Array of Object References=Get L_AR_NewAssets()
    NewItem:"Object Reference (by ref)"=Get L_NewAsset() -> (L_NewAsset:Object Reference)
   )
   n17 = Set ParentClass(execute=@n2.NewEnumerator1, ParentClass=/Game/Environments/Blockout/Blueprints/Props/BP_Blockout_Prop_Medium_A.BP_Blockout_Prop_Medium_A_C)
   n18 = Set ParentClass(execute=@n2.NewEnumerator2, ParentClass=/Game/Environments/Blockout/Blueprints/Props/BP_Blockout_Prop_Large_A.BP_Blockout_Prop_Large_A_C)
   n19 = Set ParentClass(execute=@n2.NewEnumerator3, ParentClass=/Script/FortniteGame.BuildingStaticGeo)
  ]}
 CreateNewAssetData {nodes [
   n1 = CreateNewAssetData!() -> (then, InAsset:Object Reference) {is=title K2Node_FunctionEntry}
   > n2 = Branch(Condition="Equal (Class)"(A=Get ParentClass(), B=&BuildingStaticGeo))
   > n3 = Set L_BP_Prefix(L_BP_Prefix:String=BSG_) -> (Output_Get:String)
   n4 = Sequence!(execute=[@n2.else, @n3.then]) -> (then_0, then_1, then_2:Exec)
   n5 = Split(
    SourceString="To String (Name)"(InName=CreateAssetData!(InAsset=@n1, bAllowBlueprintClass=false) -> (
      ReturnValue_PackageName:Name
      ReturnValue_PackagePath:Name
      ReturnValue_AssetName:Name
      ReturnValue_AssetClassPath:Top Level Asset Path Structure
     ))
    InStr=/
    SearchDir=FromEnd
   )
   > n6 = Set L_SourcePath(L_SourcePath:String=@n5) -> (Output_Get:String)
   > Set NewPath(NewPath:String=Replace(SourceString=@n6, From=Meshes, To=Blueprints)) -> (Output_Get:String)
   n7 = Set L_SourceName(execute=@n4.then_1, L_SourceName:String=@n5.RightS) -> (Output_Get:String)
   > n8 = Branch(Condition=MatchesWildcard(SourceString=@n7, Wildcard=SM_))
   n9 = Get L_BP_Prefix() -> (L_BP_Prefix:String)
   > Set NewName(NewName:String=Replace(SourceString=Get L_SourceName() -> (L_SourceName:String), From=SM_, To=LeftChop(SourceString=@n9, Count=1))) -> (Output_Get:String)
   Set NewName(execute=@n8.else, NewName=Append(A=Get L_BP_Prefix(), B=Get L_SourceName()))
   Return Node!(
    execute=@n4.then_2
    PackageName:String=Append!(A=Get NewPath() -> (NewPath:String), B=/, C:String=Get NewName() -> (NewName:String)) -> (ReturnValue)
   )
  ]}
 EventGraph {links [], nodes []}
]
variables [
 AR_SelectedMeshes {type="array<StaticMesh>", default []}
 NewName {type=string, default=""}
 NewPath {type=string, default=""}
 ParentClass {type=Class, default=null}
]
